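# Material function that layers an animated plasma effect with a clear-coat shading model over input material attributes for the Horned Judgment Midgard player body.
# Plasma combines scrolling flowmap distortion, bump-offset noise and turbulence, hue-shifted beams and Fresnel on the second UV set, guided by packed FX textures.
# A static switch adds neck masking and normal detail, while a component mask drives blending.
# Quality and shading-path switches simplify costly layers on lower-end paths.
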MaterialFunction /Game/Characters/Player/Male/Medium/Bodies/M_MED_HornedJudgment_Midgard/Materials/MF_HornedJudgment_Plasma
graph {nodes [
  n1 = LinearInterpolate(A=NamedRerouteUsage(), B=NamedRerouteUsage(), Alpha=NamedRerouteUsage())
  n2 = Multiply(A=@n1, B=OneMinus(None=ScalarParameter() {group=HornedJudgment PlasmaFX, parameterName=TubeLinesIntensity, sortPriority=55}))
  n3 = NamedRerouteUsage()
  n4 = Constant()
  n5 = SetMaterialAttributes!(
   MaterialAttributes=NamedRerouteUsage()
   EmissiveColor=StaticSwitchParameter(True=LinearInterpolate(A=@n1, B=@n2, Alpha=@n3) {constA=1}, False=NamedRerouteUsage()) {group=HornedJudgment PlasmaFX, parameterName=useNeckMask, sortPriority=3}
   BaseColor=@n4
   Normal=NamedRerouteUsage()
   CustomData1=@n4
   CustomData0=Multiply(A=Constant() {r=1}, B=ScalarParameter() {defaultValue=1, group=HornedJudgment PlasmaFX, parameterName=ClearCoatMultiplier, sortPriority=111}) {constB=0.7}
   ShadingModel=ShadingModel() {shadingModel=MSM_ClearCoat}
   Metallic=Multiply(
    A=NamedRerouteUsage()
    B=ScalarParameter() {defaultValue=0.7, group=HornedJudgment PlasmaFX, parameterName=MetallicMultiplier, sortPriority=110}
   ) {constB=0.7}
  ) {attributeSetTypes [
    #1
    #2
    #3
    #4
    #5
    #6
    #7
   ]}
  n6 = NamedRerouteUsage()
  n7 = LinearInterpolate(A=@n6, Alpha=NamedRerouteUsage())
  FunctionOutput(None=BlendMaterialAttributes(
    A=NamedRerouteUsage()
    B=@n5
    Alpha=StaticSwitchParameter(True=@n7, False=@n6) {group=HornedJudgment PlasmaFX, parameterName=useNeckMask, sortPriority=3}
   ))
  n8 = GetMaterialAttributes(None=FunctionInput() {inputType=FunctionInput_MaterialAttributes}) {attributeGetTypes [#1, #3, #7]}
  NamedRerouteDeclaration(None=@n8.MaterialAttributes) {name=InMA, nodeColor=<rgba 1 0 0.0469 1>}
  NamedRerouteDeclaration(None=@n8.EmissiveColor) {name=InEmissive, nodeColor=<rgba 0 1 0.3084 1>}
  n9 = TextureCoordinate()
  n10 = ScalarParameter() {defaultValue=1, group=HornedJudgment EmissiveFX, parameterName=EmissiveGradientTile, sortPriority=20}
  n11 = Time()
  n12 = ScalarParameter() {defaultValue=0.3, group=HornedJudgment EmissiveFX, parameterName=EmissiveGradientSpeed, sortPriority=21}
  n13 = ComponentMask(None=Add(A=Multiply(A=@n9, B=@n10), B=Multiply(A=@n11, B=@n12) {constB=0.3})) {g=true}
  n14 = Sine(None=Divide(A=Frac(None=@n13)))
  NamedRerouteDeclaration(None=Power(Base=Abs(None=@n14)) {constExponent=6}) {name=EmissiveGradient, nodeColor=<rgba 0 0.4844 1 1>}
  NamedRerouteDeclaration(None=@n8.Normal) {name=InNormal, nodeColor=<rgba 0 0.9766 1 1>}
  NamedRerouteDeclaration(None=Multiply(A=Time(), B=ScalarParameter() {defaultValue=1, group=HornedJudgment PlasmaFX, parameterName=GeneralScrollSpeed, sortPriority=5})) {name=PlasmaGeneralTime, nodeColor=<rgba 1 0.1641 0 1>}
  n15 = ScalarParameter() {group=HornedJudgment PlasmaFX, parameterName=DropOffsetX, sortPriority=41}
  n16 = ScalarParameter() {group=HornedJudgment PlasmaFX, parameterName=DropOffsetY, sortPriority=40}
  n17 = Add(A=Divide(A=AppendVector(A=@n15, B=@n16)) {constB=10}, B=TextureCoordinate() {coordinateIndex=1})
  n18 = Multiply(
   A=NamedRerouteUsage()
   B=Divide(A=ScalarParameter() {defaultValue=1, group=HornedJudgment PlasmaFX, parameterName=GeneralDistortion, sortPriority=38}) {constB=100}
  ) {constB=0.015}
  n19 = QualitySwitch(Default=NamedRerouteDeclaration(None=Add(A=@n17, B=@n18)) {name=UVDistorted, nodeColor=<rgba 0 1 0.7422 1>}, Low=@n17)
  n20 = AppendVector(
   A=NamedRerouteDeclaration(None=ScalarParameter() {defaultValue=1, group=HornedJudgment PlasmaFX, parameterName=EffectsUVTileX, sortPriority=5}) {name=UVTileXValue, nodeColor=<rgba 1 0.0469 0 1>}
   B=ScalarParameter() {defaultValue=1, group=HornedJudgment PlasmaFX, parameterName=EffectsUVTileY, sortPriority=6}
  )
  NamedRerouteDeclaration(None=Multiply(
    A=ShadingPathSwitch(Default=@n19, Mobile=@n17) {desc=Migrated from FeatureLevelSwitch}
    B=NamedRerouteDeclaration(None=@n20) {name=UVTileValues, nodeColor=<rgba 0 1 0.1797 1>}
   )) {name=UVTiled, nodeColor=<rgba 0 0.7891 1 1>}
  n21 = NamedRerouteUsage()
  n22 = Sine(None=Divide(A=Frac(None=@n21)))
  NamedRerouteDeclaration(None=ComponentMask(None=Abs(None=@n22)) {r=true}) {name=CurveGradient, nodeColor=<rgba 0.9453 0 1 1>}
  n23 = Add(A=TextureCoordinate() {coordinateIndex=1, uTiling=5, vTiling=5}, B=Multiply(A=NamedRerouteUsage(), B=Constant2Vector() {g=1.5, r=0.2}))
  NamedRerouteDeclaration(None=ComponentMask(None=TextureSample(UVs=@n23) {samplerType=SAMPLERTYPE_LinearColor, texture=/Game/Global/Textures/Default/Flowmaps/T_SwirlFlow_02}.RGB) {g=true, r=true}) {name=BasicDistortion, nodeColor=<rgba 0.9219 1 0 1>}
  n24 = ScalarParameter() {defaultValue=0.5, group=HornedJudgment PlasmaFX, parameterName=FresnelFXMin}
  n25 = Multiply(A=NamedRerouteUsage()) {constB=0.2}
  n26 = Constant() {r=0.02}
  n27 = NamedRerouteUsage()
  n28 = NamedRerouteUsage()
  n29 = Multiply(A=ComponentMask(None=NamedRerouteUsage()) {g=true, r=true}) {constB=0.1}
  n30 = NamedRerouteUsage()
  n31 = Constant2Vector() {g=2, r=3}
  n32 = NamedRerouteUsage()
  n33 = Add(A=Add(A=Multiply(A=@n30, B=@n31), B=Multiply(A=@n32) {constB=0.2}), B=Multiply(A=NamedRerouteUsage(), B=Constant2Vector() {g=1, r=0.1}))
  n34 = TextureSample(UVs=BumpOffset(Coordinate=Add(A=@n29, B=@n33), Height=Constant() {r=-3})) {texture=/Game/Global/Textures/Default/Noise/T_Tiling_Noise_01}
  n35 = Constant() {r=0.5}
  n36 = Add(
   A=Multiply(A=Add(A=@n28) {constB=0.2}) {constB=0.2}
   B=ShadingPathSwitch(Default=QualitySwitch(Default=@n34.R, Low=@n35), Deferred=@n35, Mobile=@n35) {desc=Migrated from FeatureLevelSwitch}
  )
  n37 = LinearInterpolate(
   A=Multiply(A=MaterialFunctionCall(Hue Shift Percentage=@n26, Texture=@n27) {materialFunction=&HueShift}.Result) {constB=0.4}
   B=Saturate(None=Multiply(A=@n27) {constB=1.5})
   Alpha=LinearInterpolate(
    A=ScalarParameter() {group=HornedJudgment PlasmaFX, parameterName=BackGroundNoiseMin, sortPriority=35}
    B=ScalarParameter() {defaultValue=1, group=HornedJudgment PlasmaFX, parameterName=BackGroundNoiseMax, sortPriority=36}
    Alpha=Saturate(None=@n36)
   )
  )
  n38 = NamedRerouteUsage()
  n39 = Multiply(A=NamedRerouteUsage(), B=NamedRerouteUsage())
  n40 = ScalarParameter() {defaultValue=1, group=HornedJudgment PlasmaFX, parameterName=BeamIntensity, sortPriority=8}
  n41 = NamedRerouteUsage()
  n42 = LinearInterpolate(
   A=LinearInterpolate(A=LinearInterpolate(A=@n25, B=@n37, Alpha=@n38) {constAlpha=1}, B=NamedRerouteUsage(), Alpha=Multiply(A=@n39, B=@n40))
   B=Multiply(A=NamedRerouteUsage()) {constB=3}
   Alpha=Multiply(
    A=Multiply(A=@n41) {constB=0.2}
    B=ScalarParameter() {defaultValue=1, group=HornedJudgment PlasmaFX, parameterName=BeamIntensity, sortPriority=8}
   )
  )
  n43 = ScalarParameter() {defaultValue=2, group=HornedJudgment PlasmaFX, parameterName=FresnelFXMax}
  n44 = NamedRerouteUsage()
  n45 = Multiply(
   A=LinearInterpolate(A=Multiply(A=@n42, B=@n24) {constB=3}, B=Multiply(A=@n42, B=@n43) {constB=2}, Alpha=Saturate(None=@n44))
   B=NamedRerouteUsage()
  )
  n46 = ScalarParameter() {defaultValue=1, group=HornedJudgment PlasmaFX, parameterName=ColorAlongNeck, sortPriority=77}
  n47 = TextureCoordinate() {coordinateIndex=1}
  n48 = Multiply(
   A=SmoothStep(
    Min=ScalarParameter() {group=HornedJudgment PlasmaFX, parameterName=ColorAlongNeckPositionMin, sortPriority=79}
    Max=ScalarParameter() {defaultValue=1, group=HornedJudgment PlasmaFX, parameterName=ColorAlongNeckPositionMax, sortPriority=80}
    Value=ComponentMask(None=@n47) {g=true}
   )
   B=ScalarParameter() {group=HornedJudgment PlasmaFX, parameterName=ColorAlongNeckMultiplier, sortPriority=78}
  ) {constB=2}
  NamedRerouteDeclaration(None=LinearInterpolate(A=@n45, B=Multiply(A=@n45, B=@n46) {constB=0}, Alpha=Saturate(None=@n48))) {name=Plasma, nodeColor=<rgba 0.3828 1 0 1>}
  n49 = Saturate(None=DotProduct(A=PixelNormalWS(), B=CameraVectorWS()))
  NamedRerouteDeclaration(None=Multiply(A=@n49, B=@n49)) {name=Fresnel, nodeColor=<rgba 0.5234 0 1 1>}
  n50 = MaterialFunctionCall(Float3=NamedRerouteUsage()) {materialFunction=&BreakOutFloat3Components}
  n51 = ScalarParameter() {defaultValue=1, group=HornedJudgment PlasmaFX, parameterName=TubeNormalLinesIntensity, sortPriority=56}
  n52 = Multiply(A=@n50.R, B=@n51)
  n53 = MaterialFunctionCall(Float3=NamedRerouteUsage()) {materialFunction=&BreakOutFloat3Components}
  n54 = Multiply(A=@n50.G, B=@n51)
  n55 = LinearInterpolate(B=@n50.B, Alpha=@n51) {constA=1}
  n56 = NamedRerouteUsage()
  n57 = QualitySwitch(
   Default=MaterialFunctionCall(R=Add(A=@n52, B=@n53.R), G=Add(A=@n54, B=@n53.G), B=Multiply(A=@n55, B=@n53.B)) {materialFunction=&AppendMany}.RGB
   Low=@n56
   Medium=@n56
  )
  NamedRerouteDeclaration(None=StaticSwitchParameter(True=ShadingPathSwitch(Default=@n57, Deferred=@n56, Mobile=@n56) {desc=Migrated from FeatureLevelSwitch}, False=NamedRerouteUsage()) {group=HornedJudgment PlasmaFX, parameterName=useNeckMask, sortPriority=3}) {name=CustomNormal, nodeColor=<rgba 1 0 0.3047 1>}
  n58 = AppendVector(A=Constant() {r=1}, B=NamedRerouteUsage())
  n59 = Divide(A=NamedRerouteUsage(), B=@n58)
  n60 = TextureCoordinate() {coordinateIndex=1}
  n61 = Sine(None=Divide(A=Frac(None=@n60)))
  n62 = OneMinus(None=ComponentMask(None=Abs(None=@n61)) {r=true})
  n63 = ScalarParameter() {group=HornedJudgment PlasmaFX, parameterName=FlamesPvtAroundNeck, sortPriority=100}
  n64 = AppendVector(
   A=Constant() {r=0.5}
   B=Add(
    A=ScalarParameter() {defaultValue=0.1, group=HornedJudgment PlasmaFX, parameterName=FlameSizePVT, sortPriority=83}
    B=Multiply(A=@n62, B=@n63) {constB=0.5}
   )
  )
  n65 = NamedRerouteUsage()
  n66 = BumpOffset(
   Coordinate=Subtract(A=Add(A=@n59, B=@n64), B=Divide(A=@n64, B=@n58) {desc=Scale the center offset by the scaling factor}) {desc=Re offset to maintain center}
   Height=LinearInterpolate(Alpha=Saturate(None=@n65)) {constB=0.3}
  ) {constCoordinate=1}
  n67 = NamedRerouteUsage()
  NamedRerouteDeclaration(None=Saturate(None=TextureSample(UVs=@n66, Tex=@n67) {constCoordinate=1, samplerType=SAMPLERTYPE_LinearColor}.R)) {name=Beam, nodeColor=<rgba 0.3594 1 0 1>}
  n68 = Time()
  n69 = Divide(A=ScalarParameter() {defaultValue=1, group=HornedJudgment PlasmaFX, parameterName=EmissiveVariationSpeed, sortPriority=31}) {constB=10}
  n70 = Add(A=Frac(None=Multiply(A=@n68, B=@n69) {constB=0.1}), B=NamedRerouteUsage())
  n71 = Add(A=Divide(A=Sine(None=@n70))) {constB=0.5}
  NamedRerouteDeclaration(None=LinearInterpolate(
    A=ScalarParameter() {group=HornedJudgment PlasmaFX, parameterName=EmissiveVariationMin, sortPriority=31}
    B=ScalarParameter() {defaultValue=1, group=HornedJudgment PlasmaFX, parameterName=EmissiveVariationMax, sortPriority=30}
    Alpha=Saturate(None=@n71)
   ) {constA=0.3, constB=1.5}) {name=EmissiveTubeVariation, nodeColor=<rgba 1 0.1641 0 1>}
  n72 = TextureObjectParameter() {
   constCoordinate=1
   group=HornedJudgment PlasmaFX
   parameterName=UV2 FX
   samplerType=SAMPLERTYPE_LinearColor
   sortPriority=3
   texture=/Game/Characters/Player/Male/Medium/Bodies/M_MED_HornedJudgment_Valhalla/Textures/T_HornedJudgment_Valhalla_UV2_Body_FX
  }
  NamedRerouteDeclaration(None=@n72) {name=UV2FX, nodeColor=<rgba 0 0.1562 1 1>}
  n73 = ScalarParameter() {defaultValue=8, group=HornedJudgment EmissiveFX, parameterName=EmissiveGradientMin, sortPriority=22}
  n74 = ScalarParameter() {defaultValue=15, group=HornedJudgment EmissiveFX, parameterName=EmissiveGradientMax, sortPriority=23}
  n75 = NamedRerouteUsage()
  NamedRerouteDeclaration(None=Multiply(A=NamedRerouteUsage(), B=LinearInterpolate(A=@n73, B=@n74, Alpha=@n75) {constA=8, constB=15}) {constB=8}) {name=Panning Emissive, nodeColor=<rgba 1 0.3281 0 1>}
  n76 = TextureObjectParameter() {
   constCoordinate=1
   group=HornedJudgment PlasmaFX
   parameterName=UV2 Normal
   samplerType=SAMPLERTYPE_Normal
   sortPriority=3
   texture=/Game/Characters/Player/Male/Medium/Bodies/M_MED_HornedJudgment_Valhalla/Textures/T_HornedJudgment_Valhalla_UV2_Body_N
  }
  n77 = TextureSample(Tex=@n76) {
   constCoordinate=1
   samplerType=SAMPLERTYPE_Normal
   texture=&T_HornedJudgment_Valhalla_UV2_Body_N
  }
  NamedRerouteDeclaration(None=@n77.RGB) {name=uv2Normal, nodeColor=<rgba 1 0.8672 0 1>}
  NamedRerouteDeclaration(None=@n8.Metallic) {name=InMetallic, nodeColor=<rgba 0.4531 1 0 1>}
  n78 = TextureSample(Tex=@n72) {constCoordinate=1, samplerType=SAMPLERTYPE_LinearColor}
  NamedRerouteDeclaration(None=@n78.A) {name=UV2FX_A_CurveGrad, nodeColor=<rgba 1 1 1 1>}
  NamedRerouteDeclaration(None=@n78.G) {name=UV2FX_G_TubeLines, nodeColor=<rgba 0 1 0 1>}
  NamedRerouteDeclaration(None=@n78.B) {name=UV2FX_B_ColorVariation, nodeColor=<rgba 0 0 1 1>}
  NamedRerouteDeclaration(None=StaticComponentMaskParameter(None=FunctionInput() {inputName=TAMask, sortPriority=1}) {defaultG=true, group=HornedJudgment PlasmaFX, parameterName=EmissivePlasmaMask, sortPriority=0}) {name=TAMask, nodeColor=<rgba 1 0 0.9375 1>}
  n79 = Noise!(World Position=Time(), FilterWidth) {bTiling=true, bTurbulence=false, levelScale=1, levels=1, noiseFunction=NOISEFUNCTION_GradientTex, outputMin=0, repeatSize=64}
  n80 = Divide(A=ScalarParameter() {defaultValue=1, group=HornedJudgment PlasmaFX, parameterName=FlameVariationSpeed, sortPriority=82}) {constB=10}
  n81 = Add(A=Frac(None=Multiply(A=@n79, B=@n80) {constB=0.1}), B=NamedRerouteUsage())
  n82 = Add(A=Divide(A=Sine(None=@n81))) {constB=0.5}
  NamedRerouteDeclaration(None=LinearInterpolate(
    A=ScalarParameter() {defaultValue=0.8, group=HornedJudgment PlasmaFX, parameterName=FlameVariationMin, sortPriority=80}
    B=ScalarParameter() {defaultValue=1, group=HornedJudgment PlasmaFX, parameterName=FlameVariationMax, sortPriority=81}
    Alpha=Saturate(None=@n82)
   ) {constA=0.3, constB=1.5}) {name=FlameSizeTubeVariation, nodeColor=<rgba 1 0 0.0776 1>}
  n83 = NamedRerouteUsage()
  n84 = Constant() {r=0.01}
  n85 = Multiply(A=@n83) {constB=4}
  n86 = Multiply(A=Power(Base=NamedRerouteUsage(), Exp=ScalarParameter() {defaultValue=2.5, group=HornedJudgment PlasmaFX, parameterName=BeamSize, sortPriority=8}) {constExponent=2.5}) {constB=1.5}
  NamedRerouteDeclaration(None=LinearInterpolate(
    A=Multiply(A=@n83) {constB=0.8}
    B=MaterialFunctionCall(Hue Shift Percentage=@n84, Texture=@n85) {materialFunction=&HueShift}.Result
    Alpha=Saturate(None=@n86)
   ) {constB=0.8}) {name=BeamColored, nodeColor=<rgba 0.6172 0 1 1>}
  n87 = Power(Base=NamedRerouteUsage())
  n88 = LinearInterpolate(
   B=Power(Base=NamedRerouteUsage()) {constExponent=1}
   Alpha=ScalarParameter() {defaultValue=1, group=HornedJudgment PlasmaFX, parameterName=GenericFresnal, sortPriority=15}
  ) {constA=1}
  NamedRerouteDeclaration(None=Saturate(None=Multiply(A=@n87, B=@n88))) {name=AdvancedFresnel, nodeColor=<rgba 0 1 0.5078 1>}
  n89 = NamedRerouteUsage()
  n90 = Multiply(A=@n89, B=@n89)
  n91 = Multiply(A=@n90, B=@n90) {constB=1.8}
  n92 = NamedRerouteUsage()
  n93 = Constant2Vector() {g=2, r=0.2}
  n94 = NamedRerouteUsage()
  n95 = Add(A=Add(A=NamedRerouteUsage(), B=Multiply(A=@n92, B=@n93)), B=ComponentMask(None=Multiply(A=@n94) {constB=0.3}) {g=true, r=true})
  n96 = LinearInterpolate(Alpha=NamedRerouteUsage()) {constA=-10, constB=-5}
  n97 = Subtract(
   A=TextureSample(UVs=BumpOffset(Coordinate=@n95, Height=@n96)) {texture=/Game/Global/Textures/Default/Caustics/T_Turbulence_Seamless}.R
   B=Multiply(A=NamedRerouteUsage()) {constB=0.4}
  )
  n98 = Multiply(A=@n91, B=Multiply(A=Saturate(None=@n97)) {constB=4}) {constA=1}
  n99 = ShadingPathSwitch(Default=QualitySwitch(Default=@n98, Low=@n91, High=@n98, Medium=@n91), Deferred=@n91, Mobile=@n91) {desc=Migrated from FeatureLevelSwitch}
  NamedRerouteDeclaration(None=Multiply(A=NamedRerouteUsage(), B=Saturate(None=@n99))) {name=FrontLayer, nodeColor=<rgba 0 0.5781 1 1>}
  n100 = NamedRerouteUsage()
  n101 = NamedRerouteUsage()
  n102 = Constant2Vector() {g=1.5, r=0.2}
  n103 = TextureSample(UVs=Add(A=Multiply(A=@n100) {constB=1.5}, B=Multiply(A=@n101, B=@n102))) {samplerType=SAMPLERTYPE_LinearGrayscale, texture=/Game/Global/Textures/Default/Noise/T_SwirlNoise_N_height}
  n104 = Constant() {r=0.5}
  NamedRerouteDeclaration(None=ShadingPathSwitch(Default=QualitySwitch(Default=@n103.R, Low=@n104, High=@n104, Medium=@n104), Deferred=@n104, Mobile=@n104) {desc=Migrated from FeatureLevelSwitch}) {name=GlobalNoise, nodeColor=<rgba 0 0.7187 1 1>}
  n105 = Add(
   A=TextureCoordinate() {coordinateIndex=1}
   B=AppendVector(A=Constant(), B=ScalarParameter() {group=HornedJudgment PlasmaFX, parameterName=NeckMaskYPosition, sortPriority=4})
  )
  n106 = ScalarParameter() {defaultValue=0.2, group=HornedJudgment PlasmaFX, parameterName=NeckMaskSize, sortPriority=4}
  n107 = Min(A=ComponentMask(None=@n105) {r=true}, B=Subtract(A=ComponentMask(None=@n105) {g=true}, B=@n106) {constA=0.3, constB=0.2})
  n108 = OneMinus(None=@n105)
  n109 = Min(A=ComponentMask(None=@n108) {r=true}, B=Subtract(A=ComponentMask(None=@n108) {g=true}, B=@n106) {constA=0.3, constB=0.2})
  NamedRerouteDeclaration(None=Step(X=Min(A=@n107, B=@n109)) {constX=0.1}) {name=NeckMask, nodeColor=<rgba 0 1 0.2031 1>}
 ]}
