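# Master material for the Super Level skin of the Horned Judgment Midgard character, supporting skeletal meshes and Nanite.
# The shared 2023 character PreFX function feeds a plasma function, enabled by default through a static switch, then the PostFX function.
# The Season 27 Super Level function finishes the material attributes, with original emissive enabled.
# Its FX mask uses the parent's tech-art mask, or optionally a separate texture via static switch.
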Material /Game/Characters/Player/Male/Medium/Bodies/M_MED_HornedJudgment_Midgard/Skins/SuperLevel/Materials/M_HornedJudgment_Midgard_SL
graph {
 properties {
  bUseMaterialAttributes=true
  bUsedWithMeshDeformer=true
  bUsedWithMorphTargets=true
  bUsedWithNanite=true
  bUsedWithSkeletalMesh=true
  pixelDepthOffsetMode=PDOM_Legacy
  shadingModel=MSM_FromMaterialExpression
  usedShadingModels=Default Lit | Subsurface | Clear Coat | Subsurface Profile | Cloth
 }
 nodes [
  n1 = MaterialFunctionCall() {materialFunction=&MF_CharacterParent_2023_PreFX}
  n2 = MaterialFunctionCall!(In=@n1.Result, TAMask=@n1.Result) {materialFunction=/Game/Characters/Player/Male/Medium/Bodies/M_MED_HornedJudgment_Midgard/Materials/MF_HornedJudgment_Plasma}
  n3 = TextureSampleParameter2D() {
   group=SuperLevel
   parameterName=TAMaskSL
   samplerSource=SSM_Wrap_WorldGroupSettings
   samplerType=SAMPLERTYPE_Masks
   sortPriority=101
   texture=/Game/Characters/Player/Male/Medium/Bodies/M_MED_HornedJudgment_Valhalla/Skins/SuperLevel/Textures/T_HornedJudgment_Valhalla_FaceAcc_TechArtFX_Mask
  }
  MaterialFunctionCall!(
   InMA=MaterialFunctionCall(
    Material=StaticSwitchParameter(True=@n2.Result, False=@n1.Result) {defaultValue=true, group=SuperLevel, parameterName=UsePlasma, sortPriority=50}
    M=@n1.M
    Pre Skinned Local Position=@n1.M
    Distance Blend=@n1.M
    Fresnel=@n1.M
   ) {materialFunction=&MF_CharacterParent_2023_PostFX}.Result
   FXMask=StaticSwitchParameter(True=@n3.RGB, False=@n1.Tech Art Mask) {group=SuperLevel, parameterName=UseExtraTAMaskForSL, sortPriority=100}
   UseOriginalEmissive=StaticBool() {value=true}
   PreSkinPos=@n1.Pre skinned Local Position
  ) {materialFunction=/Game/Characters/Player/Common/MaterialFunctions/Superlevel_S27/MF_S27SL}
 ]
}
